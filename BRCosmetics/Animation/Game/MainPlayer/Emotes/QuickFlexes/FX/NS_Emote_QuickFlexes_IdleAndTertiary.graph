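# Niagara system for the idle and tertiary part of the QuickFlexes emote, drawing glow sprites attached to a rocket skeletal mesh.
# Glow spawns two bursts split into six partitions, each following a bone or socket index with an offset per spawn group.
# Nested partition masks select staggered stepped curves, switching each beep color from red to green before fading.
# Base color has a front-end variant; six SmokeIdleA emitters list no renderers.
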NiagaraSystem /BRCosmetics/Animation/Game/MainPlayer/Emotes/QuickFlexes/FX/NS_Emote_QuickFlexes_IdleAndTertiary
dependencies {
 usedDataInterfaces [
  &NiagaraDataInterfaceArrayInt32
  &NiagaraDataInterfaceColorCurve
  &NiagaraDataInterfaceSkeletalMesh
 ]
 usedDynamicInputs [&RandomRangeFloat]
 usedModules [
  &self:SystemSpawnScript.SystemScriptSource.SystemScriptGraph.NiagaraNodeAssignment_1.SetVariables_#1
  &SystemState
  &self:SystemSpawnScript.SystemScriptSource.SystemScriptGraph.NiagaraNodeAssignment_2.SetVariables_#2
 ]
}
emitters {
 Glow {
  emitterUpdateScript [
   &EmitterState(Life Cycle Mode=::NewEnumerator1, Loop Behavior=::NewEnumerator1, Loop Duration Mode=::NewEnumerator1, UseLoopDelay=true)
   &SpawnBurst_Instantaneous(Spawn Count=6, Use Spawn Probability=false, UseLoopCountLimit=false)
   &SpawnBurst_Instantaneous(Spawn Count=6, Spawn Group=1, Use Spawn Probability=false, UseLoopCountLimit=false) {module=SpawnBurst_Instantaneous001}
  ]
  particleSpawnScript [
   &PartitionParticles(Number Of Partitions=6)
   &InitializeParticle(
    Color=&MakeLinearColorFromVectorAndFloat(
     "Vector (RGB)"=&FN_FrontEndVector3Modifier(
      Gameplay=&MakeVectorFromLinearColorRGB(Color=<rgba 0.276 0.0661 0 1>)
      Front End=&MakeVectorFromLinearColorRGB(Color=<rgba 0.9271 0.2219 0 1>)
     )
     "Float (Alpha)"=0.539
    )
    Lifetime=7.3
    Sprite Rotation Mode=::NewEnumerator1
    Sprite Size=<18 20>
    Sprite Size Mode=::NewEnumerator4
    UsePositionOffset=true
    Write Material Random=true
   ) {module=InitializeParticle001}
   /SetVariables(Particles.VisibilityTag=&SpawnGroup()) {isSetParameters=true}
   &CameraOffset(Camera Offset Amount=3)
  ]
  particleUpdateScript [
   &ParticleState()
   &SkeletalMeshLocation(
    Bone / Socket Index=&SelectIntFromArray(Direct Array Index=$Particles.PartitionParticles.Partition, Int Selection Array=DataInterface() {propertyValues {IntData [0,1,2,3,4,5]}})
    Bone Sampling Mode=::NewEnumerator4
    MaskSpawnGroup=false
    Sampled Position Offset=&MaskVectorBySpawnGroup(Masked Vector=<4 0 0>, Passthrough Vector=<-15 0 0>)
    Skeletal Mesh=$System.RocketLoction
    Use Sampled Mesh Attribute as Ribbon ID=false
   )
   &SolveForcesAndVelocity(Clamp Velocity=false, Limit Acceleration=false)
   &ScaleColor(ScaleA=true, ScaleRGB=true)
   /SetVariables(
    Particles.BeepColor:LinearColor=&MaskLinearColor(
     Mask Source=1
     Mask Parameter=$Particles.PartitionParticles.Partition
     Mask=0
     Masked Linear Color=&ColorFromCurve(ColorCurve=DataInterface() {propertyValues {
        AlphaCurve {keys [<at 0.9275 1 RCIM_Cubic>,<at 1 0 RCIM_Cubic>]}
        BlueCurve {keys [<at 0.043 0 RCIM_Constant>,<at 0.2388 0 RCIM_Constant>]}
        GreenCurve {keys [<at 0.043 0.0167 RCIM_Constant>,<at 0.2388 2 RCIM_Constant>]}
        RedCurve {keys [<at 0.043 2 RCIM_Constant>,<at 0.2388 0.211 RCIM_Constant>]}
       }})
     Passthrough Linear Color=&MaskLinearColor(
      Mask Source=1
      Mask Parameter=$Particles.PartitionParticles.Partition
      Mask=1
      Masked Linear Color=&ColorFromCurve(ColorCurve=DataInterface() {propertyValues {
         AlphaCurve {keys [<at 0.9275 1 RCIM_Cubic>,<at 1 0 RCIM_Cubic>]}
         BlueCurve {keys [<at 0.0078 0 RCIM_Constant>,<at 0.3223 0 RCIM_Constant>]}
         GreenCurve {keys [<at 0.0078 0.0167 RCIM_Constant>,<at 0.3223 2 RCIM_Constant>]}
         RedCurve {keys [<at 0.0078 2 RCIM_Constant>,<at 0.3223 0.211 RCIM_Constant>]}
        }})
      Passthrough Linear Color=&MaskLinearColor(
       Mask Source=1
       Mask Parameter=$Particles.PartitionParticles.Partition
       Mask=2
       Masked Linear Color=&ColorFromCurve(ColorCurve=DataInterface() {propertyValues {
          AlphaCurve {keys [<at 0.9275 1 RCIM_Cubic>,<at 1 0 RCIM_Cubic>]}
          BlueCurve {keys [<at 0.1917 0 RCIM_Constant>,<at 0.4507 0 RCIM_Constant>]}
          GreenCurve {keys [<at 0.1917 0.0167 RCIM_Constant>,<at 0.4507 2 RCIM_Constant>]}
          RedCurve {keys [<at 0.1917 2 RCIM_Constant>,<at 0.4507 0.211 RCIM_Constant>]}
         }})
       Passthrough Linear Color=&MaskLinearColor(
        Mask Source=1
        Mask Parameter=$Particles.PartitionParticles.Partition
        Mask=3
        Masked Linear Color=&ColorFromCurve(ColorCurve=DataInterface() {propertyValues {
           AlphaCurve {keys [<at 0.9275 1 RCIM_Cubic>,<at 1 0 RCIM_Cubic>]}
           BlueCurve {keys [<at 0.0017 0 RCIM_Constant>,<at 0.3995 0 RCIM_Constant>]}
           GreenCurve {keys [<at 0.0017 0.0167 RCIM_Constant>,<at 0.3995 2 RCIM_Constant>]}
           RedCurve {keys [<at 0.0017 2 RCIM_Constant>,<at 0.3995 0.211 RCIM_Constant>]}
          }})
        Passthrough Linear Color=&MaskLinearColor(
         Mask Source=1
         Mask Parameter=$Particles.PartitionParticles.Partition
         Mask=4
         Masked Linear Color=&ColorFromCurve(ColorCurve=DataInterface() {propertyValues {
            AlphaCurve {keys [<at 0.9275 1 RCIM_Cubic>,<at 1 0 RCIM_Cubic>]}
            BlueCurve {keys [<at 0.3793 0 RCIM_Constant>,<at 0.5072 0 RCIM_Constant>]}
            GreenCurve {keys [<at 0.3793 0.0167 RCIM_Constant>,<at 0.5072 2 RCIM_Constant>]}
            RedCurve {keys [<at 0.3793 2 RCIM_Constant>,<at 0.5072 0.211 RCIM_Constant>]}
           }})
         Passthrough Linear Color=&ColorFromCurve(ColorCurve=DataInterface() {propertyValues {
            AlphaCurve {keys [<at 0.9275 1 RCIM_Cubic>,<at 1 0 RCIM_Cubic>]}
            BlueCurve {keys [<at 0.097 0 RCIM_Constant>,<at 0.2582 0 RCIM_Constant>]}
            GreenCurve {keys [<at 0.097 0.0167 RCIM_Constant>,<at 0.2582 2 RCIM_Constant>]}
            RedCurve {keys [<at 0.097 2 RCIM_Constant>,<at 0.2582 0.211 RCIM_Constant>]}
           }})
        )
       )
      )
     )
    )
    Particles.BeepSize:Vector2f=&Multiply_Vector2D_ByFloat(Float=5, Vector2D=<1 1>)
   ) {isSetParameters=true}
  ]
  rendererClasses [&NiagaraSpriteRendererProperties]
  renderers [
   {rendererClass=&NiagaraSpriteRendererProperties}
   {rendererClass=&NiagaraSpriteRendererProperties, rendererIndex=1}
  ]
 }
 SmokeIdleA {rendererClasses [], renderers []}
 SmokeIdleA001 {rendererClasses [], renderers []}
 SmokeIdleA002 {rendererClasses [], renderers []}
 SmokeIdleA003 {rendererClasses [], renderers []}
 SmokeIdleA004 {rendererClasses [], renderers []}
 SmokeIdleA005 {rendererClasses [], renderers []}
}
summary {
 systemName=NS_Emote_QuickFlexes_IdleAndTertiary
 emitters [
  {emitterName=SmokeIdleA, rendererClasses []}
  {emitterName=SmokeIdleA001, rendererClasses []}
  {emitterName=SmokeIdleA002, rendererClasses []}
  {emitterName=SmokeIdleA003, rendererClasses []}
  {emitterName=SmokeIdleA004, rendererClasses []}
  {emitterName=SmokeIdleA005, rendererClasses []}
  {emitterName=Glow, rendererClasses [&NiagaraSpriteRendererProperties]}
 ]
 userVariables [
  User.bIsFrontEnd {type=NiagaraBool}
  User.bIsFrontEndPreview {type=NiagaraBool}
 ]
}
