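# Niagara effect for the QuickFlexes emote that emits smoke and a soft ribbon trail from skeletal mesh sockets chosen by User.SocketNumber, timed by User.Loop.
# Smoke001 runs two tapering spawn groups, tagged for renderer visibility across two sprite renderers, with curve-driven size, fade and material parameters.
# SoftTrail builds a widening anim-trail ribbon between that socket index and the next, tinted by time of day.
# Flash and SmokeExplode list no renderers.
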NiagaraSystem /BRCosmetics/Animation/Game/MainPlayer/Emotes/QuickFlexes/FX/NS_Emote_QuickFlexes_Ribbons
dependencies {
 usedDataInterfaces [&NiagaraDataInterfaceCurve, &NiagaraDataInterfaceVectorCurve, &NiagaraDataInterfaceSkeletalMesh]
 usedDynamicInputs []
 usedModules [
  &SystemState
  &self:SystemSpawnScript.SystemScriptSource.SystemScriptGraph.NiagaraNodeAssignment_2.SetVariables_#1
 ]
}
emitters {
 Flash {rendererClasses [], renderers []}
 Smoke001 {
  emitterUpdateScript [
   &EmitterState(Life Cycle Mode=::NewEnumerator1, Loop Behavior=::NewEnumerator1, Loop Duration=$User.Loop, UseLoopDelay=true)
   &SpawnRate(
    SpawnRate=&FloatFromCurve(FloatCurve=DataInterface() {propertyValues {Curve {keys [<at 0 1>,<at 1 0.4688>]}}}, CurveIndex=$Emitter.NormalizedLoopAge, Scale Curve=60)
    Use Spawn Probability=false
   )
   &SpawnRate(
    SpawnGroup=1
    SpawnRate=&FloatFromCurve(FloatCurve=DataInterface() {propertyValues {Curve {keys [<at 0 1>,<at 1 0.4688>]}}}, CurveIndex=$Emitter.NormalizedLoopAge, Scale Curve=25)
    Use Spawn Probability=false
   ) {module=SpawnRate001}
  ]
  particleSpawnScript [
   &InitializeParticle(
    Color=<rgba 0.5255 0.5255 0.5255 1>
    Lifetime=&UniformRangedFloat(Minimum=0.2, Maximum=0.4)
    Sprite Rotation Angle=&UniformRangedFloat(Minimum=0, Maximum=1)
    Sprite Rotation Mode=::NewEnumerator1
    Sprite Size Mode=::NewEnumerator1
    Uniform Sprite Size Max=6
    Uniform Sprite Size Min=4
    UsePositionOffset=false
    Write Material Random=true
   ) {module=InitializeParticle001}
   /SetVariables(Particles.VisibilityTag=&SpawnGroup()) {isSetParameters=true}
   &CameraOffset(Camera Offset Amount=&RandomRangeFloat(Minimum=10, Maximum=15))
   &AddVelocity(Velocity=<0 0 -10>)
   &AddVelocity(Module.Use Falloff Distance=false, Override Seed=false, Velocity Mode=::NewEnumerator1) {module=AddVelocity001}
   &SkeletalMeshLocation(
    Bone / Socket Index=$User.SocketNumber
    MaskSpawnGroup=false
    Mesh Sampling Type=::NewEnumerator5
    Skeletal Mesh=$System.RocketLoction
    Socket Sampling Mode=::NewEnumerator5
   )
   &ShapeLocation(MaskSpawnGroup=false, Override Seed=false, Rotation Coordinate Space=::Simulation, Sphere Radius=0, UseOwnerScale=true)
   /SetVariables(Particles.VelocityLong:Vector3f=<0 0 -1>) {isSetParameters=true}
  ]
  particleUpdateScript [
   &ParticleState()
   &SpriteRotationRate(Rotation Rate=&Multiply_Float(
     A=$Particles.InitSpriteRotRate
     B=&FloatFromCurve(
      FloatCurve=DataInterface() {propertyValues {Curve {keys [
          {arriveTangent=-1.9214, interpMode=RCIM_Cubic, leaveTangent=-1.9214, tangentMode=RCTM_User, time=-0.0046, value=1.0113}
          {arriveTangent=-0.2496, interpMode=RCIM_Cubic, leaveTangent=-0.2496, tangentMode=RCTM_User, time=0.9896, value=0.16}
         ]}}}
      CurveIndex=$Emitter.Age
     )
    ))
   &ScaleColor(
    Scale Alpha=&FloatFromCurve(FloatCurve=DataInterface() {propertyValues {Curve {keys [<at 0 0>,<at 0.0727 1>,<at 0.7791 1>,<at 1 0>]}}})
    ScaleA=true
    ScaleRGB=true
   ) {module=ScaleColor001}
   &Drag(Drag=0.2382, Use Linear Drag=true, Use Rotational Drag=false)
   &SolveForcesAndVelocity(Clamp Velocity=false, Limit Acceleration=false)
   &ScaleSpriteSize(Scale Factor=&Vector2DFromFloat(Value=&FloatFromCurve(FloatCurve=DataInterface() {propertyValues {Curve {keys [
          {arriveTangent=-3.33, interpMode=RCIM_Cubic, leaveTangent=-3.33, tangentMode=RCTM_User, time=0, value=0.3524}
          {arriveTangent=2.2217, interpMode=RCIM_Cubic, leaveTangent=2.2217, tangentMode=RCTM_User, time=0.4, value=2.1131}
          {arriveTangent=-0.6199, interpMode=RCIM_Cubic, leaveTangent=-0.6199, tangentMode=RCTM_User, time=1.0011, value=3.0398}
         ]}}})))
   &DynamicMaterialParameters(
    Index 0 Param 1=&FloatFromCurve(FloatCurve=DataInterface() {propertyValues {Curve {keys [<at 0 0 RCIM_Cubic>, {arriveTangent=0.4364, tangentMode=RCTM_User, time=1.0011, value=0.2}]}}})
    Index 0 Param 2=&FloatFromCurve(FloatCurve=DataInterface() {propertyValues {Curve {keys [
         {arriveTangent=0.0453, interpMode=RCIM_Cubic, tangentMode=RCTM_Break, time=0, value=-1}
         {arriveTangent=0.0604, interpMode=RCIM_Cubic, leaveTangent=0.0604, tangentMode=RCTM_User, time=1, value=-1}
        ]}}})
    Index 0 Param 3=&FloatFromCurve(FloatCurve=DataInterface() {propertyValues {Curve {keys [{interpMode=RCIM_Cubic, leaveTangent=4.6779, time=0.2851, value=0}]}}})
    Index 1 Param 1=&FloatFromCurve(
     FloatCurve=DataInterface() {propertyValues {Curve {keys [<at 0 0>,<at 0.25 1>]}}}
     Scale Curve=&RandomRangeFloat(Minimum=2, Maximum=1)
    )
    Index 1 Param 2=&FloatFromCurve(
     FloatCurve=DataInterface() {propertyValues {Curve {keys [
         {interpMode=RCIM_Cubic, leaveTangent=3.7422, tangentMode=RCTM_Break, time=0, value=0.2}
         {interpMode=RCIM_Cubic, leaveTangent=0.9066, time=1, value=2}
        ]}}}
     Scale Curve=&RandomRangeFloat(Minimum=0.5, Maximum=1)
    )
    Index 1 Param 3=&FloatFromCurve(FloatCurve=DataInterface() {propertyValues {Curve {keys [
         {interpMode=RCIM_Cubic, leaveTangent=2.8445, tangentMode=RCTM_Break, time=0, value=0.2}
         {interpMode=RCIM_Cubic, leaveTangent=63.2978, time=0.25, value=1}
        ]}}})
    Index 1 Param 4=&FloatFromCurve(
     FloatCurve=DataInterface() {propertyValues {Curve {keys [<at 0 1>,<at 1 0.8>]}}}
     Scale Curve=&RandomRangeFloat(Minimum=0.5, Maximum=1)
    )
    Index 2 Param 1=&FloatFromCurve(FloatCurve=DataInterface() {propertyValues {Curve {keys [{interpMode=RCIM_Cubic, leaveTangent=3.2309, tangentMode=RCTM_Break, time=0.1, value=0}, <at 1 1>]}}})
    Index 2 Param 2=$Particles.MaterialRandom
    Index 2 Param 3=&FloatFromCurve(FloatCurve=DataInterface() {propertyValues {Curve {keys [
         {interpMode=RCIM_Cubic, tangentMode=RCTM_Break, time=0, value=1}
         {interpMode=RCIM_Cubic, leaveTangent=11.1233, tangentMode=RCTM_Break, time=1, value=2}
        ]}}})
    Index 2 Param 4=0.75
    Module.Param0WriteEnabled=true
    Module.Param0WriteEnabled1=true
    Module.Param0WriteEnabled2=true
    Module.Param1WriteEnabled=false
    Module.Param1WriteEnabled1=true
    Module.Param1WriteEnabled2=true
    Module.Param2WriteEnabled=true
    Module.Param2WriteEnabled1=true
    Module.Param2WriteEnabled2=true
    Module.Param3WriteEnabled=true
    Module.Param3WriteEnabled1=true
    Module.Param3WriteEnabled2=true
    Write Parameter Index 1=true
    Write Parameter Index 2=true
   )
   /SetVariables(Particles.VelocitySize:Vector2f=/Niagara/DynamicInputs/ScaleByCurve/ScaleVector2DByCurve(
     InitialValue:Vector2f=&RandomRangeVector2D(Minimum=<3 6>, Maximum=<5 10>)
     ScaleCurve:NiagaraDataInterfaceVectorCurve=DataInterface() {propertyValues {
       XCurve {
        defaultValue=0
        keys [
         {arriveTangent=-3.33, interpMode=RCIM_Cubic, leaveTangent=-3.33, tangentMode=RCTM_User, time=0, value=0.3524}
         {arriveTangent=2.2217, arriveTangentWeight=0.1333, interpMode=RCIM_Cubic, leaveTangent=2.2217, tangentMode=RCTM_User, time=0.4, value=2.75}
         {
          arriveTangent=-0.6199
          arriveTangentWeight=0.2004
          interpMode=RCIM_Cubic
          leaveTangent=-0.6199
          tangentMode=RCTM_User
          time=1.0011
          value=3.4809
         }
        ]
       }
       YCurve {
        defaultValue=0
        keys [
         {arriveTangent=-3.33, interpMode=RCIM_Cubic, leaveTangent=-3.33, leaveTangentWeight=0.3337, tangentMode=RCTM_User, time=0, value=0.3524}
         {
          arriveTangent=2.2217
          arriveTangentWeight=0.1333
          interpMode=RCIM_Cubic
          leaveTangent=2.2217
          leaveTangentWeight=0.2004
          tangentMode=RCTM_User
          time=0.4
          value=2.75
         }
         {
          arriveTangent=-0.6199
          arriveTangentWeight=0.2358
          interpMode=RCIM_Cubic
          leaveTangent=-0.6199
          tangentMode=RCTM_User
          time=1.0011
          value=3.4809
         }
        ]
       }
       ZCurve {keys []}
      }}
     CurveIndex:NiagaraFloat=$Particles.NormalizedAge
    )) {isSetParameters=true}
  ]
  rendererClasses [&NiagaraSpriteRendererProperties]
  renderers [
   {rendererClass=&NiagaraSpriteRendererProperties}
   {rendererClass=&NiagaraSpriteRendererProperties, rendererIndex=1}
  ]
 }
 SmokeExplode {rendererClasses [], renderers []}
 SoftTrail {
  emitterUpdateScript [
   &EmitterState(Life Cycle Mode=::NewEnumerator1, Loop Behavior=::NewEnumerator1, Loop Duration=$User.Loop, UseLoopDelay=true)
   &FN_SpawnAnimTrail()
  ]
  particleSpawnScript [
   &FN_InitializeAnimTrailBoneIndecies(
    First Bone or Socket Index:NiagaraInt32=&Add_Integer(A=1, B=$User.SocketNumber)
    Second Bone or Socket Index:NiagaraInt32=&Add_Integer(A=0, B=$User.SocketNumber)
   )
   &FN_InitializeAnimTrail(
    AnimTrailWidth=1.9759
    Color=&MakeLinearColorFromVectorAndFloat(
     "Vector (RGB)"=&MakeVectorFromLinearColorRGB(Color=&FNTimeOfDayColor(Night Color=<rgba 0.1475 0.1406 1 1>, Morning Color=<rgba 1 0.5806 0.0014 1>, Day Color=<rgba 1 1 1 1>, Evening Color=<rgba 1 0.5421 0 1>))
     "Float (Alpha)"=&FNTimeOfDayScalar(Night Scale=0.7, Morning Scale=0.7, Evening Scale=0.7)
    )
    Position Offset=<0 0 -7.0481>
    SkeletalMesh=$System.RocketLoction
   )
   &AddVelocity(Velocity=<0 0 0.4387>, Velocity Speed Scale=92.7816)
  ]
  particleUpdateScript [
   &FN_UpdateAnimTrail(SkeletalMesh=$System.RocketLoction) {module=FN_UpdateAnimTrail001}
   &InheritSourceMovement(ApplyForce=false, ApplyPosition=true, ApplyPositionFalloff=false)
   &SolveForcesAndVelocity(Clamp Velocity=false, Limit Acceleration=false)
   &ScaleRibbonWidth(Ribbon Width Scale=&FloatFromCurve(FloatCurve=DataInterface() {propertyValues {Curve {keys [<at 0 0.3>,<at 1 4.3981>]}}}))
   &DynamicMaterialParameters(
    Index 0 Param 1=-0.144
    Index 0 Param 3=-0.9249
    Index 0 Param 4=&FloatFromCurve(FloatCurve=DataInterface() {propertyValues {Curve {keys [<at 0 1>,<at 1 0>]}}})
    Module.Index0Float0Write=true
    Module.Index0Float1Write=true
    Module.Index0Float2Write=true
    Module.Index0Float3Write=true
   )
  ]
  rendererClasses [&NiagaraRibbonRendererProperties]
  renderers [{rendererClass=&NiagaraRibbonRendererProperties}]
 }
}
summary {
 systemName=NS_Emote_QuickFlexes_Ribbons
 emitters [
  {emitterName=SmokeExplode, rendererClasses []}
  {emitterName=Flash, rendererClasses []}
  {emitterName=Smoke001, rendererClasses [&NiagaraSpriteRendererProperties]}
  {emitterName=SoftTrail, rendererClasses [&NiagaraRibbonRendererProperties]}
 ]
 userVariables [
  User.Loop {type=NiagaraFloat, defaultValue=3.1469, description=""}
  User.RibbonIDOffset {type=NiagaraInt32}
  User.SocketNumber {type=NiagaraInt32}
 ]
}
